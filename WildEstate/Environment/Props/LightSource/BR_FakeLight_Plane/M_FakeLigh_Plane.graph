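# Additive, unlit, two-sided material that fakes a light glow on a plane.
# A vertical gradient with power falloff is tinted by time-of-day colors, scaled by intensity, then passed through shared FX emissive adjustments.
# A Pulse built from two floored time sines at different rates blinks falloff and intensity between two states.
# Near and far distance fades feed a soft depth fade for opacity, blended by the Confidential collection parameter.
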Material /WildEstate/Environment/Props/LightSource/BR_FakeLight_Plane/M_FakeLigh_Plane
graph {
 properties {bAutomaticallySetUsageInEditor=false, blendMode=BLEND_Additive, shadingModel=MSM_Unlit, twoSided=true, usedShadingModels=Unlit}
 nodes [
  CollectionParameter() {collection=/WildEstate/TechArt/MPCs/MPC_WildEstate_TA, parameterName=Confidential}
  n1 = ScalarParameter() {
   defaultValue=1000
   desc=Distance at which light starts to disappear when get nearby.
   group=Opacity
   parameterName=NearFadeDistance
   sortPriority=0
  }
  n2 = MaterialFunctionCall!(Blend Range=@n1, Start Offset=Constant()) {materialFunction=/Engine/Functions/Engine_MaterialFunctions03/Particles/Distance_Blend}
  n3 = ScalarParameter() {
   defaultValue=2000
   desc=Distance at which light starts to disappear when when get away from it.
   group=Opacity
   parameterName=FarFadeDistance
   sortPriority=1
  }
  n4 = OneMinus(None=MaterialFunctionCall(Blend Range=@n3, Start Offset=Multiply(A=@n3) {constB=-1}) {materialFunction=&Distance_Blend}.Result)
  LinearInterpolate(
   B=DepthFade(Opacity=Min(A=@n2.Result, B=@n4)) {fadeDistanceDefault=25}
   Alpha=CollectionParameter() {collection=&MPC_WildEstate_TA, parameterName=Confidential}
  )
  n5 = Time()
  n6 = Sine(None=Divide(A=@n5) {constB=0.5})
  n7 = Sine(None=Divide(A=@n5) {constB=3})
  n8 = Min(A=Add(A=Floor(None=@n6)), B=Add(A=Floor(None=@n7)))
  LinearInterpolate(Alpha=@n8) {constA=0.5, constB=5}
  n9 = Power(
   Base=MaterialFunctionCall(UV Channel=TextureCoordinate()) {materialFunction=&LinearGradient}.VGradient
   Exp=LinearInterpolate(
    A=ScalarParameter() {defaultValue=5, group=Light, parameterName=FalloffMin, sortPriority=3}
    B=ScalarParameter() {defaultValue=2.5, group=Light, parameterName=FalloffMax, sortPriority=3}
    Alpha=NamedRerouteUsage()
   )
  ) {constExponent=5}
  n10 = MaterialFunctionCall(
   Color_Morning=VectorParameter() {parameterName=Color_Morning, defaultValue=<rgba 0.1043 1 0 1>}.RGB
   Color_Day=VectorParameter() {parameterName=Color_Day, defaultValue=<rgba 0.0347 1 0 1>}.RGB
   Color_Evening=VectorParameter() {parameterName=Color_Evening, defaultValue=<rgba 0.0543 1 0 1>}.RGB
   Color_Night=VectorParameter() {parameterName=Color_Night, defaultValue=<rgba 0.0397 1 0 1>}.RGB
  ) {materialFunction=&Color_Time_of_Day}
  n11 = ScalarParameter() {defaultValue=0.2, group=Light, parameterName=Intensity, sortPriority=0}
  n12 = ScalarParameter() {defaultValue=0.66, group=Light, parameterName=IntensityPulse, sortPriority=0}
  n13 = NamedRerouteUsage()
  MaterialFunctionCall(Emissive=Multiply(A=Multiply(A=@n9, B=@n10.Result), B=LinearInterpolate(A=@n11, B=@n12, Alpha=@n13))) {materialFunction=&MF_FX_EmissiveAdjustments}
  n14 = VectorParameter() {group=Light, parameterName=Color Day, sortPriority=1, defaultValue=<rgba 1 0.8228 0.7011 1>}
  n15 = VectorParameter() {group=Light, parameterName=Color Night, sortPriority=2, defaultValue=<rgba 1 0.8228 0.7011 1>}
  MaterialFunctionCall(Color_Morning=@n14.RGB, Color_Day=@n14.RGB, Color_Evening=@n15.RGB, Color_Night=@n15.RGB) {materialFunction=&Color_Time_of_Day}
  NamedRerouteDeclaration(None=@n8) {name=Pulse, nodeColor=<rgba 1 0.5625 0 1>}
 ]
}
